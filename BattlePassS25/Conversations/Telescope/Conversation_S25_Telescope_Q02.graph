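# Season 25 Telescope Q02 conversation in which the player hands a Kinetic Prism to NPC Trace.
# The auto-selected main-questline opener requires Trace's entry tag, an active Q02, obj0 complete and obj1 incomplete.
# Three linear speech nodes lead to two task outputs: one completes Q02 obj1, the other completes Q01_Urgent obj0 and requires that quest to be active.
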ConversationDatabase /BattlePassS25/Conversations/Telescope/Conversation_S25_Telescope_Q02
entry_points [{tag=Conversation.Entry.Telescope.Q02.Trace, destinations [#1]}]
nodes [
 ConversationEntryPointNode_0 {
  id=#1
  class=ConversationEntryPointNode
  outputs [#2]
  properties {entryTag {tagName=Conversation.Entry.Telescope.Q02.Trace}}
 }
 Speech_C_3 {
  id=#2
  class=Speech_C
  outputs [#3]
  properties {softTaskIcon=&T_UI_NPC_MainQuest, generalConfig {defaultMessage="Hey, another Prism! Thanks. "}}
  sub_nodes [
   ConversationChoiceNode_1 {
    class=ConversationChoiceNode
    properties {
     defaultChoiceDisplayText=Talk to Trace
     choiceTags {gameplayTags [{tagName=Conversation.Choice.AutoSelectIfValid}, {tagName=Conversation.Choice.Category.MainQuestline}]}
    }
   }
   FortConversationRequirement_ControllerRequirement_3 {
    class=FortConversationRequirement_ControllerRequirement
    properties {
     nodeName=Require: NPC Has Conversation Tag
     participantId {tagName=Conversation.Participant.NPC}
     requirement {class=FortControllerRequirement_NPCHasConversationEntryTag, properties {queryTag {tagName=Conversation.Entry.Trace}}}
    }
   }
   FortConversationRequirement_ControllerRequirement_0 {
    class=FortConversationRequirement_ControllerRequirement
    properties {
     nodeName=Require: Is Quest Active
     requirement {
      class=FortControllerRequirement_IsQuestActive
      properties {questDefinition=/BattlePassS25/Items/QuestItems/Telescope/quest_s25_Telescope_Q02}
     }
    }
   }
   FortConversationRequirement_ControllerRequirement_2 {
    class=FortConversationRequirement_ControllerRequirement
    properties {
     nodeName=Require: Has Completed Quest Objective
     requirement {
      class=FortControllerRequirement_HasCompletedQuestObjective
      properties {objectiveBackendName=quest_s25_Telescope_Q02_obj0, questDefinition=&quest_s25_Telescope_Q02}
     }
    }
   }
   FortConversationRequirement_ControllerRequirement_1 {
    class=FortConversationRequirement_ControllerRequirement
    properties {
     nodeName=Require: Has Completed Quest Objective
     requirement {
      class=FortControllerRequirement_HasCompletedQuestObjective
      properties {
       bInverseCondition=true
       objectiveBackendName=quest_s25_Telescope_Q02_obj1
       questDefinition=&quest_s25_Telescope_Q02
      }
     }
    }
   }
  ]
 }
 Speech_C_28 {
  id=#3
  class=Speech_C
  outputs [#4]
  properties {generalConfig {defaultMessage=Our researchers think these Kinetic Prisms were made for refracting light.}}
  sub_nodes [ConversationChoiceNode_0 {class=ConversationChoiceNode, properties {defaultChoiceDisplayText=Hand over Prism}}]
 }
 Speech_C_0 {
  id=#4
  class=Speech_C
  outputs [#5, #6]
  properties {generalConfig {defaultMessage=Like some sort of lens...}}
  sub_nodes [ConversationChoiceNode_4 {
    class=ConversationChoiceNode
    properties {defaultChoiceDisplayText=Continue, choiceTags {gameplayTags [{tagName=Athena.Quests.S25.Conversations.Trace.Choice.Q01}]}}
   }]
 }
 FortConversationTaskNode_CompleteQuestObjectiveByBackendName_1 {
  id=#6
  class=FortConversationTaskNode_CompleteQuestObjectiveByBackendName
  properties {
   objectiveBackendName=quest_s25_Telescope_Q01_Urgent_obj0
   questDefinition=/BattlePassS25/Items/QuestItems/Telescope/quest_s25_Telescope_Q01_Urgent
  }
  sub_nodes [FortConversationRequirement_ControllerRequirement_4 {
    class=FortConversationRequirement_ControllerRequirement
    properties {
     nodeName=Require: Is Quest Active
     requirement {
      class=FortControllerRequirement_IsQuestActive
      properties {questDefinition=&quest_s25_Telescope_Q01_Urgent}
     }
    }
   }]
 }
 FortConversationTaskNode_CompleteQuestObjectiveByBackendName_0 {
  id=#5
  class=FortConversationTaskNode_CompleteQuestObjectiveByBackendName
  properties {objectiveBackendName=quest_s25_Telescope_Q02_obj1, questDefinition=&quest_s25_Telescope_Q02}
 }
]
